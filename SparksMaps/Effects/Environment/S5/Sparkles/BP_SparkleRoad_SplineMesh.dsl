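; Sparkle road spline-mesh Blueprint for environment effects, pairing a static mesh with a Niagara sparkle system.
; Its construction script pushes the Time Offset variable into the mesh materials as a scalar parameter.
; It also sets the Niagara component's OffsetScalar Vector2 parameter from the Particle Offset Scalar variable, so placed instances can vary timing and particle offset.
; BeginPlay is empty.
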(event EventBeginPlay)

(fn ConstructionScript
  (Rendering|Material|SetScalarParameterValueonMaterials (Variables|Default|GetStaticMesh) "Time Offset" (Variables|Default|GetTime%20Offset))
  (Niagara|SetNiagaraVariableByString(Vector2) (Variables|Default|GetNS_SparkleRoad) "OffsetScalar" (Variables|Default|GetParticleOffsetScalar)))
