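# Unlit additive sprite material for Niagara and particle sparkle flares with a hot center, drawn before depth of field.
# A radial mask forms a whitened core and soft halo, added to a tinted flare texture.
# Soften Hot Spot blends toward the tinted texture alone before emissive adjustments.
# Opacity combines particle alpha, depth-faded except on mobile, with a power-curved camera-distance fade from Near_Camera_Fade.
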Material /SparksMaps/Effects/Environment/S5/Sparkles/M_Particle_Flare_HotCenter_NoFlicker_FarFade_BeforeDOF
graph {
 properties {
  bUsedWithNiagaraSprites=true
  bUsedWithParticleSprites=true
  blendMode=BLEND_Additive
  pixelDepthOffsetMode=PDOM_Legacy
  shadingModel=MSM_Unlit
  translucencyPass=MTP_BeforeDOF
  usedShadingModels=Unlit
 }
 nodes [
  n1 = Constant() {r=0.5}
  n2 = SphereMask(A=TextureCoordinate(), B=@n1, Radius=@n1) {hardnessPercent=0}
  n3 = ParticleColor()
  n4 = Multiply(A=Multiply(A=Power(Base=@n2) {constExponent=25}) {constB=2}, B=LinearInterpolate(B=@n3.RGB) {constA=1, constAlpha=0.7})
  n5 = Power(Base=@n2) {constExponent=4}
  n6 = TextureSampleParameter2D() {parameterName=FlareMask, texture=/Game/Effects/Fort_Effects/Textures/Flares/T_Flare_Spikeball}
  n7 = Multiply(A=Add(A=Multiply(A=@n5) {constB=0.15}, B=@n6.RGB), B=@n3.RGB)
  MaterialFunctionCall(Emissive=LinearInterpolate(A=Add(A=@n4, B=@n7), B=Multiply(A=@n6.RGB, B=@n3.RGB), Alpha=ScalarParameter() {parameterName=Soften Hot Spot})) {materialFunction=&MF_FX_EmissiveAdjustments}
  n8 = DepthFade(Opacity=@n3.A, FadeDistance=ScalarParameter() {defaultValue=30, parameterName=DepthFade})
  n9 = Power(
   Base=MaterialFunctionCall(Distance=ScalarParameter() {defaultValue=3000, parameterName=NearFade}) {materialFunction=&Near_Camera_Fade}.Result
   Exp=ScalarParameter() {defaultValue=8, parameterName=Near Fade Curve Power}
  ) {constExponent=8}
  Multiply(
   A=ShadingPathSwitch(Default=@n8, Deferred=@n8, Forward=@n8, Mobile=@n3.A)
   B=LinearInterpolate(A=Constant(), B=Constant() {r=1}, Alpha=Clamp(None=@n9))
  )
 ]
}
