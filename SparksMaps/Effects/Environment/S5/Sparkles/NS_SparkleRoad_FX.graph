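# Niagara environment effect that scatters white, cyan and purple sprite sparkles along a user-supplied spline, as a sparkle road.
# One emitter combines a burst with two spawn rates, gated by user Spawn Probability, and tags particles by spawn group alongside three sprite renderers.
# Local modules set each particle's spline U, then drift it along the spline with random offsets scaled by OffsetScalar.
# Alpha fades by both spline position and age; many older look modules remain disabled.
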NiagaraSystem /SparksMaps/Effects/Environment/S5/Sparkles/NS_SparkleRoad_FX
dependencies {
 usedDataInterfaces [
  &NiagaraDataInterfaceCurve
  &NiagaraDataInterfaceArrayColor
  &NiagaraDataInterfaceColorCurve
  &NiagaraDataInterfaceSpriteRendererInfo
 ]
 usedDynamicInputs [&Subtract_Float]
 usedModules [
  &self:GetSplineLengthSystem
  &SystemState
  &self:SystemSpawnScript.SystemScriptSource.SystemScriptGraph.NiagaraNodeAssignment_1.SetVariables_#1
  &self:SystemSpawnScript.SystemScriptSource.SystemScriptGraph.NiagaraNodeAssignment_0.SetVariables_#2
 ]
}
emitters {SparkleRoadSpline {
  emitterUpdateScript [
   &EmitterState()
   &SpawnBurst_Instantaneous(Spawn Count=20, Spawn Probability=$User.Spawn Probability, Use Spawn Probability=true, UseLoopCountLimit=false)
   &SpawnRate(Spawn Probability=$User.Spawn Probability, SpawnRate=40, Use Spawn Probability=true) {module=SpawnRate001}
   &SpawnRate(Spawn Probability=$User.Spawn Probability, SpawnGroup=1, SpawnRate=20, Use Spawn Probability=true)
  ]
  particleSpawnScript [
   /SetVariables(Particles.VisibilityTag=&MakeCustomIntFromBool(
     Bool=&SetBoolByIntComparison(A=&SpawnGroup(), B=1, Comparison Type=::NewEnumerator4)
     True INT=1
     False INT=0
    )) {isSetParameters=true}
   /SetVariables(Particles.VisTag2:NiagaraInt32=&MakeCustomIntFromBool(
     Bool=&SetBoolByIntComparison(A=&SpawnGroup(), B=0, Comparison Type=::NewEnumerator4)
     True INT=1
     False INT=0
    )) {isSetParameters=true}
   &InitializeParticle(
    Color=&SelectLinearColorFromArray(
     Array Sampling Mode=::NewEnumerator0
     Direct Array Index=?
     Color Selection Array=DataInterface() {propertyValues {ColorData [<rgba 1.5 1.5 1.5 1>, <rgba 0.6641 1.4444 1.5 1>, <rgba 1.2984 0.7422 1.5 1>]}}
     Random Seed=0
    )
    Lifetime Max=3
    Lifetime Mode=::NewEnumerator1
    Sprite Rotation Mode=::NewEnumerator2
    Sprite Size Mode=::NewEnumerator1
    Uniform Sprite Size Max=28
    Uniform Sprite Size Min=20
    UsePositionOffset=false
    Write Material Random=true
   )
   &self:SetSplineU(NumberofPos=15) {module=SetSplineU}
  ]
  particleUpdateScript [
   &ParticleState()
   &ScaleSpriteSize(
    Scale Sprite Size Mode=::NewEnumerator0
    Uniform Scale Factor=&RandomRangeFloat(Minimum=3, Maximum=5, Evaluation Type=::NewEnumerator1)
   ) {module=ScaleSpriteSize002}
   &SpriteRotationRate() {enabled=false}
   &ScaleSpriteSize(
    Uniform Curve Index=$Particles.SplineU
    Uniform Curve Sprite Scale=DataInterface() {propertyValues {Curve {keys [
        {arriveTangent=4.6611, interpMode=RCIM_Cubic, leaveTangent=4.6611, tangentMode=RCTM_User, time=0, value=0}
        {arriveTangent=1.4883, interpMode=RCIM_Cubic, leaveTangent=1.1668, tangentMode=RCTM_Break, time=0.1, value=0.25}
        <at 1 0.9>
       ]}}}
   ) {enabled=false, module=ScaleSpriteSize001}
   &ScaleSpriteSize(Uniform Curve Sprite Scale=DataInterface() {propertyValues {Curve {keys [
        {interpMode=RCIM_Cubic, leaveTangent=0.6142, tangentMode=RCTM_Break, time=0, value=0.75}
        {interpMode=RCIM_Cubic, leaveTangent=-0.0715, tangentMode=RCTM_Break, time=0.5, value=0.9875}
        <at 1 0.7375 RCIM_Cubic>
       ]}}}) {enabled=false}
   &ScaleColor(
    Scale Alpha=&Multiply_Float(
     A=&FloatFromCurve(
      FloatCurve=DataInterface() {propertyValues {Curve {keys [
          {interpMode=RCIM_Cubic, tangentMode=RCTM_User, time=0, value=0.2}
          {interpMode=RCIM_Cubic, tangentMode=RCTM_User, time=0.25, value=1}
          {tangentMode=RCTM_User, time=1, value=0}
         ]}}}
      CurveIndex=$Particles.SplineU
     )
     B=0.5
    )
    Scale RGB=&Multiply_Vector(
     A=&MakeVectorFromLinearColorRGB(Color=&ColorFromCurve(
       ColorCurve=DataInterface() {propertyValues {
         AlphaCurve {keys [<at 0 0.9 RCIM_Cubic>,<at 1 0 RCIM_Cubic>]}
         BlueCurve {keys [<at 0.0104 0>,<at 0.0401 0>,<at 0.1654 0.1294>]}
         GreenCurve {keys [<at 0.0104 3.2>,<at 0.0401 0.6>,<at 0.1654 0.1328>]}
         RedCurve {keys [<at 0.0104 16>,<at 0.0401 6>,<at 0.1654 0.1396>]}
        }}
       CurveIndex=$Particles.SplineU
      ))
     B=<1 1 1>
    )
    ScaleA=true
    ScaleRGB=true
   ) {enabled=false}
   &self:SampleAndMoveSplineU(
    Offset:Vector2f=&Multiply_Vector2D(A=&RandomRangeVector2D(Minimum=<-800 -75>, Maximum=<800 10>), B=$User.OffsetScalar)
    Spline:NiagaraDataInterfaceSpline=$User.Spline
    UPerSecond:NiagaraFloat=&RandomRangeFloat(Minimum=0, Maximum=0.005)
   ) {module=SampleAndMoveSplineU}
   /SetVariables(Particles.Position=&AddVectorToPosition(Position=$Particles.Position, Vector=<0 0 0>)) {isSetParameters=true}
   &SubUVAnimation(
    "Sprite Renderer "=DataInterface() {propertyValues {SpriteRenderer=&self:SparkleRoadSpline_0.NiagaraSpriteRendererProperties_0}}
    Start Frame Range Override=&RandomRangeInt(Minimum=0, Maximum=0)
    SubUV Animation Mode:ENiagara_SubUVLookupModeV2=::NewEnumerator1
    SubUV Curve Index:NiagaraFloat=$Particles.SplineU
    UseEndFrame=false
    UseStartFrame=false
   ) {enabled=false}
   &DynamicMaterialParameters(
    Index 0 Param 1=&FloatFromCurve(
     FloatCurve=DataInterface() {propertyValues {Curve {keys [{interpMode=RCIM_Cubic, leaveTangent=6.4752, tangentMode=RCTM_Break, time=0, value=1.6}, <at 0.5 3>]}}}
     CurveIndex=$Particles.SplineU
    )
    Index 1 Param 1=0.8
    Index 1 Param 2=1
    Index 1 Param 3=&FloatFromCurve(FloatCurve=DataInterface() {propertyValues {Curve {keys [<at 0 0>,<at 0.1 0.2>,<at 0.4 0.6>]}}}, CurveIndex=$Particles.SplineU)
    Module.Param0WriteEnabled=true
    Module.Param0WriteEnabled1=true
    Module.Param1WriteEnabled=true
    Module.Param1WriteEnabled1=true
    Module.Param2WriteEnabled=true
    Module.Param2WriteEnabled1=true
    Module.Param3WriteEnabled=true
    Module.Param3WriteEnabled1=true
    Write Parameter Index 1=true
   ) {enabled=false, module=DynamicMaterialParameters001}
   &CameraOffset(Camera Offset Amount=&FloatFromCurve(FloatCurve=DataInterface() {propertyValues {Curve {keys [<at 0 1>,<at 1 0>]}}}, CurveIndex=$Particles.SplineU, Scale Curve=0)) {enabled=false, module=CameraOffset001}
   &ScaleColor(
    Scale Alpha=&Multiply_Float(
     A=&FloatFromCurve(
      FloatCurve=DataInterface() {propertyValues {Curve {keys [{interpMode=RCIM_Cubic, tangentMode=RCTM_User, time=0, value=0}, <at 0.5 0.95 RCIM_Cubic>, {tangentMode=RCTM_User, time=1, value=0}]}}}
      Scale Curve=&MaskFloatBySpawnGroup(Masked Float=1.5, Passthrough Float=1)
     )
     B=1
    )
    Scale RGB=&MakeVectorFromLinearColorRGB(Color=$Particles.Color)
    ScaleA=true
    ScaleRGB=true
   ) {enabled=false, module=ScaleColor001}
   /SetVariables(
    Particles.CamOff2:NiagaraFloat=&FloatFromCurve(FloatCurve=DataInterface() {propertyValues {Curve {keys [<at 0 1>,<at 1 0>]}}}, CurveIndex=$Particles.SplineU, Scale Curve=-50)
    Particles.Linear Color=&MakeLinearColorFromVectorAndFloat(
     "Vector (RGB)"=&MakeVectorFromLinearColorRGB(Color=$Particles.Color)
     "Float (Alpha)"=&FloatFromCurve(FloatCurve=DataInterface() {propertyValues {Curve {keys [<at 0.8 1>,<at 1 0>]}}})
    )
    Particles.SpriteSize2=&Multiply_Vector2D_ByFloat(
     Float=&Multiply_Float(
      A=&MakeFloatFromVector2d(Vector 2D=$Particles.Initial.SpriteSize)
      B=&Multiply_Float(
       A=1.1
       B=&FloatFromCurve(
        FloatCurve=DataInterface() {propertyValues {Curve {keys [{arriveTangent=1.3931, time=0.1, value=0}, <at 0.95 2>]}}}
        CurveIndex=$Particles.SplineU
       )
      )
     )
     Vector2D=<1 1>
    )
   ) {enabled=false, isSetParameters=true}
   &ScaleColor(
    Scale Alpha=&FloatFromCurve(
     FloatCurve=DataInterface() {propertyValues {Curve {keys [
         {arriveTangent=4.8958, interpMode=RCIM_Cubic, leaveTangent=4.8958, tangentMode=RCTM_User, time=0, value=0}
         {interpMode=RCIM_Cubic, tangentMode=RCTM_User, time=0.5, value=1}
         {arriveTangent=-2.4979, tangentMode=RCTM_User, time=1, value=0}
        ]}}}
     CurveIndex=$Particles.SplineU
    )
    ScaleA=true
    ScaleRGB=true
   ) {module=ScaleColor002}
   /SetVariables(Particles.SpriteSize2=&Multiply_Vector2D(A=$Particles.SpriteSize, B=<2.5 2.5>)) {isSetParameters=true}
   &ScaleColor(
    Scale Alpha=&FloatFromCurve(FloatCurve=DataInterface() {propertyValues {Curve {keys [
         {interpMode=RCIM_Cubic, tangentMode=RCTM_User, time=0, value=0}
         {interpMode=RCIM_Cubic, tangentMode=RCTM_User, time=0.5, value=1}
         {tangentMode=RCTM_User, time=1, value=0}
        ]}}})
    ScaleA=true
    ScaleRGB=true
   ) {module=ScaleColor003}
  ]
  rendererClasses [&NiagaraSpriteRendererProperties]
  renderers [
   {rendererClass=&NiagaraSpriteRendererProperties}
   {rendererClass=&NiagaraSpriteRendererProperties, rendererIndex=1}
   {rendererClass=&NiagaraSpriteRendererProperties, rendererIndex=2}
  ]
 }}
summary {
 systemName=NS_SparkleRoad_FX
 emitters [{emitterName=SparkleRoadSpline, rendererClasses [&NiagaraSpriteRendererProperties]}]
 userVariables [
  User.End Position {type=NiagaraFloat, defaultValue=1, description=""}
  User.GlobalScaler {type=NiagaraFloat, defaultValue=1, description=""}
  User.OffsetScalar {type=Vector2f, description="", defaultValue {struct=Vector2f, value=<1 1>}}
  User.Spawn Probability {type=NiagaraFloat, defaultValue=1, description=""}
  User.Spline {
   type=NiagaraDataInterfaceSpline
   description=""
   defaultValue {
    struct=NiagaraExt_VariableValue_DataInterface
    value {
     dataInterface=&self:NiagaraDataInterfaceSpline_0
     dataInterfaceClass=/Script/Niagara.NiagaraDataInterfaceSpline
    }
   }
  }
  User.Start Position {type=NiagaraFloat, defaultValue=0, description=""}
 ]
}
